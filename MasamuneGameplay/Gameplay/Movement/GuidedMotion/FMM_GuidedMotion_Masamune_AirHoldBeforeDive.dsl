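; A guided-motion movement mode for the Masamune air hold before a dive.
; It overrides the constant-force direction query used by the guided motion.
; The direction is the owner movement component's current velocity, flattened to 2D and normalized, so the force follows the existing horizontal heading.
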(fn GetDirection_ConstantForce
  (params (GuidedMotionRuntimeData :type Fort%20Movement%20Mode%20Guided%20Motion%20Runtime%20Data%20Object%20Reference))
  (return (Math|Vector|Normalize2D(Vector) (Variables|/Script/Engine.MovementComponent|GetVelocity (Variables|FortMovementMode_BaseExtLogic|GetOwnerMovementComponent)))))
