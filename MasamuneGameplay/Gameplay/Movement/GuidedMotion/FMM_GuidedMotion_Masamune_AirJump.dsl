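; Guided motion movement mode for the Masamune air jump (air dash), applying a constant-force push.
; On activation, the server optionally removes effects granting the InAirFromSprint tag, blocking repeated dashes in one jump.
; Direction follows horizontal input acceleration at a set pitch, or actor forward without input.
; Strength scales from penalized to full by how well that input direction matches horizontal velocity; a debug print of last input remains.
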(event EventOnExtensionActivate
  (params (RuntimeContext :type MMERuntime%20Context%20Structure%20%28by%20ref%29))
  (bind _ownerpawn (Variables|FortMovementMode_BaseExtLogic|GetOwnerPawn))
  (bind _returnvalue (Ability|GetAbilitySystemComponent _ownerpawn))
  (|CallParentFunction RuntimeContext)
  ; Remove tag allowing the owning GA to activate to prevent multiple air dashes in one jump (exploit)
  (if (Networking|HasAuthority (Variables|FortMovementMode_BaseExtLogic|GetOwnerPawn))
    (if (ScalableFloat|GetValueasBool (Variables|Default|GetHF_RemoveTagOnMMEActivation))
      (Utilities|IsValid _returnvalue
        (:Is%20Valid
          (GameplayEffects|RemoveActiveEffectswithGrantedTags _returnvalue "(GameplayTags=((TagName=\"Granted.Masamune.InAirFromSprint\")))"))))))

(fn GetDirection_ConstantForce
  (params (GuidedMotionRuntimeData :type Fort%20Movement%20Mode%20Guided%20Motion%20Runtime%20Data%20Object%20Reference))
  (return (Math|Vector|SelectVector (Transformation|GetActorForwardVector (Variables|FortMovementMode_BaseExtLogic|GetOwnerPawn)) (Math|Vector|GetForwardVector (Math|Rotator|MakeRotator 0.0 (Variables|Default|GetAimJumpPitch) (.yaw (Math|Rotator|MakeRotfromX (Math|Vector|Normalize2D(Vector) (Pawn|Components|CharacterMovement|GetCurrentAcceleration (Variables|FortMovementMode_BaseExtLogic|GetOwnerMovementComponent))))))) (Math|Float|NearlyEqual(Float) (Math|Vector|VectorLengthXY (Math|Vector|Normalize (Pawn|Components|CharacterMovement|GetCurrentAcceleration (Variables|FortMovementMode_BaseExtLogic|GetOwnerMovementComponent)))) 0.0 0.01))))

(fn GetStrength_ConstantForce
  (params (GuidedMotionRuntimeData :type Fort%20Movement%20Mode%20Guided%20Motion%20Runtime%20Data%20Object%20Reference))
  (bind _ownermovementcomponent (Variables|FortMovementMode_BaseExtLogic|GetOwnerMovementComponent))
  (Development|PrintString (Variables|FortMovementMode_BaseExtLogic|GetOwnerPawn) (Utilities|String|ToString(Vector) (Pawn|Components|PawnMovement|GetLastInputVector _ownermovementcomponent)))
  (return (Math|Float|Lerp (* (ScalableFloat|GetValueatLevel (Variables|Default|GetHF_AirJumpStrength)) (ScalableFloat|GetValueatLevel (Variables|Default|GetHF_NonDirectionPenalty))) (ScalableFloat|GetValueatLevel (Variables|Default|GetHF_AirJumpStrength)) (Math|Float|MapRangeClamped (Math|Vector|DotProduct (Math|Vector|GetForwardVector (Math|Rotator|MakeRotator 0.0 (Variables|Default|GetAimJumpPitch) (.yaw (Math|Rotator|MakeRotfromX (Math|Vector|Normalize2D(Vector) (Pawn|Components|CharacterMovement|GetCurrentAcceleration _ownermovementcomponent)))))) (Math|Vector|Normalize2D(Vector) (Variables|/Script/Engine.MovementComponent|GetVelocity _ownermovementcomponent))) -1.0 1.0 0.0 1.0))))
