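; Guided motion movement mode for the Masamune charged slash, supplying the direction of its constant force.
; When the owner is falling and lacks the in-water-body tag, the force follows the forward vector of the net-synced base aim rotation.
; Otherwise, on the ground or in water, it follows the pawn's forward vector.
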(fn GetDirection_ConstantForce
  (params (GuidedMotionRuntimeData :type Fort%20Movement%20Mode%20Guided%20Motion%20Runtime%20Data%20Object%20Reference))
  (return (Math|Vector|SelectVector (Math|Vector|GetForwardVector (CharacterMovement|GetNetSyncedBaseAimRotation (Variables|FortMovementMode_BaseExtLogic|GetOwnerMovementComponent))) (Transformation|GetActorForwardVector (Variables|FortMovementMode_BaseExtLogic|GetOwnerPawn)) (and (AI|Components|NavMovement|IsFalling (Variables|FortMovementMode_BaseExtLogic|GetOwnerMovementComponent)) (not (GameplayTags|HasMatchingGameplayTag (Variables|FortMovementMode_BaseExtLogic|GetOwnerPawn) "(TagName=\"Gameplay.Status.InWaterBody\")"))))))
