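; Guided motion movement mode for the Masamune lock-on target move, overriding how long the motion can last.
; GetMaxDuration divides the owner pawn's distance to the runtime data's local move-to end location by the configured MoveToActor max speed.
; A select hardwired to true returns that travel time clamped to a short fixed window, leaving the unclamped branch unused.
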(fn GetMaxDuration
  (params (GuidedMotionRuntimeData :type Fort%20Movement%20Mode%20Guided%20Motion%20Runtime%20Data%20Object%20Reference))
  (return (Math|Float|SelectFloat (Math|Float|Clamp(Float) (Math|Float|SafeDivide (Math|Vector|Distance(Vector) (Variables|/Script/FortniteGame.FortMovementMode_GuidedMotionRuntimeData|GetLocalMoveToEndLocation GuidedMotionRuntimeData) (Transformation|GetActorLocation (Variables|FortMovementMode_BaseExtLogic|GetOwnerPawn))) (ScalableFloat|GetValueasDouble (Variables|Config-Guide-MoveToActor|GetMaxSpeed))) 0.15 0.7) (Math|Float|SafeDivide (Math|Vector|Distance(Vector) (Variables|/Script/FortniteGame.FortMovementMode_GuidedMotionRuntimeData|GetLocalMoveToEndLocation GuidedMotionRuntimeData) (Transformation|GetActorLocation (Variables|FortMovementMode_BaseExtLogic|GetOwnerPawn))) (ScalableFloat|GetValueasDouble (Variables|Config-Guide-MoveToActor|GetMaxSpeed))) true)))
